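# Physics asset for the female medium Stands ASN face-accessory mesh, covering dynamic hair bones plus head and neck.
# Five four-segment back-hair chains and four two-segment front-hair chains hang from small base capsules constrained to the head.
# Links within chains have tight swing and twist limits, while base-to-head and head-to-neck joints rotate freely.
# Segments are mostly capsules, with boxes on broader B and C back-hair sections; a sphere and capsules shape the head and neck.
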PhysicsAsset /BRCosmetics/Characters/Player/Female/Medium/Bodies/F_MED_Stands/Meshes/Parts/ASN/F_MED_Stands_ASN_FaceAcc_Physics
bodies {
 dyn_backhairA_01 [{shapeType=Capsule, length=1.5, radius=2, shapeName=dyn_backhairA_01_capsule, center=<1.8925 0 0.0001>, rotation=<rot 90 0 0>}]
 dyn_backhairA_02 [{shapeType=Capsule, length=1.5, radius=2, shapeName=dyn_backhairA_02_capsule, center=<1.8925 0 0.0001>, rotation=<rot 90 0 0>}]
 dyn_backhairA_03 [{shapeType=Capsule, length=1.5, radius=2, shapeName=dyn_backhairA_03_capsule, center=<1.8925 0 0.0001>, rotation=<rot 90 0 0>}]
 dyn_backhairA_04 [{shapeType=Capsule, length=1.5, radius=2, shapeName=dyn_backhairA_04_capsule, center=<1.8925 0 0.0001>, rotation=<rot 90 0 0>}]
 dyn_backhairA_base [{shapeType=Capsule, length=0.505, radius=0.505, shapeName=dyn_backhairA_base_capsule, rotation=<rot 0 -90 0>}]
 dyn_backhairB_01 [{shapeType=Capsule, length=2, radius=1.5, shapeName=dyn_backhairB_01_capsule, center=<1.8925 0 0.0001>, rotation=<rot 90 0 0>}]
 dyn_backhairB_02 [{shapeType=Box, extentX=4.5, extentY=1, extentZ=5.5, shapeName=dyn_backhairB_02_box, center=<2.0598 0 0>, rotation=<rot 0 0 19.9985>}]
 dyn_backhairB_03 [{shapeType=Box, extentX=4.5, extentY=1, extentZ=5.5, shapeName=dyn_backhairB_03_box, center=<2.0598 0 0>, rotation=<rot 0 0 19.9985>}]
 dyn_backhairB_04 [{shapeType=Box, extentX=4.5, extentY=1, extentZ=5.5, shapeName=dyn_backhairB_04_box, center=<2.0598 0 0>, rotation=<rot 0 0 19.9985>}]
 dyn_backhairB_base [{shapeType=Capsule, length=0.505, radius=0.505, shapeName=dyn_backhairB_base_capsule, rotation=<rot 0 -90 0>}]
 dyn_backhairC_01 [{shapeType=Box, extentX=4.5, extentY=2, extentZ=6.5, shapeName=dyn_backhairC_01_box, center=<3.4022 0 0>}]
 dyn_backhairC_02 [{shapeType=Box, extentX=4.5, extentY=2, extentZ=6.5, shapeName=dyn_backhairC_02_box, center=<3.4022 0 0>}]
 dyn_backhairC_03 [{shapeType=Box, extentX=4.5, extentY=2, extentZ=6.5, shapeName=dyn_backhairC_03_box, center=<3.4022 0 0>}]
 dyn_backhairC_04 [{shapeType=Box, extentX=4.5, extentY=1.5, extentZ=6.5, shapeName=dyn_backhairC_04_box, center=<2.0462 0 0>}]
 dyn_backhairC_base [{shapeType=Capsule, length=0.505, radius=0.505, shapeName=dyn_backhairC_base_capsule, rotation=<rot 0 -90 0>}]
 dyn_backhairD_01 [{shapeType=Capsule, length=2, radius=1.5, shapeName=dyn_backhairD_01_capsule, center=<1.8925 0 0.0001>, rotation=<rot 90 0 0>}]
 dyn_backhairD_02 [{shapeType=Capsule, length=2, radius=1.5, shapeName=dyn_backhairD_02_capsule, center=<1.8925 0 0.0001>, rotation=<rot 90 0 0>}]
 dyn_backhairD_03 [{shapeType=Capsule, length=2, radius=1.5, shapeName=dyn_backhairD_03_capsule, center=<1.8925 0 0.0001>, rotation=<rot 90 0 0>}]
 dyn_backhairD_04 [{shapeType=Capsule, length=2, radius=1.5, shapeName=dyn_backhairD_04_capsule, center=<1.8925 0 0.0001>, rotation=<rot 90 0 0>}]
 dyn_backhairD_base [{shapeType=Capsule, length=0.505, radius=0.505, shapeName=dyn_backhairD_base_capsule, rotation=<rot 0 -90 0>}]
 dyn_backhairE_01 [{shapeType=Capsule, length=2, radius=1.5, shapeName=dyn_backhairE_01_capsule, center=<1.8925 0 0.0001>, rotation=<rot 90 0 0>}]
 dyn_backhairE_02 [{shapeType=Capsule, length=2, radius=1.5, shapeName=dyn_backhairE_02_capsule, center=<1.8925 0 0.0001>, rotation=<rot 90 0 0>}]
 dyn_backhairE_03 [{shapeType=Capsule, length=2, radius=1.5, shapeName=dyn_backhairE_03_capsule, center=<1.8925 0 0.0001>, rotation=<rot 90 0 0>}]
 dyn_backhairE_04 [{shapeType=Capsule, length=2, radius=1.5, shapeName=dyn_backhairE_04_capsule, center=<1.8925 0 0.0001>, rotation=<rot 90 0 0>}]
 dyn_backhairE_base [{shapeType=Capsule, length=0.505, radius=0.505, shapeName=dyn_backhairE_base_capsule, rotation=<rot 0 -90 0>}]
 dyn_fronthairA_01 [{shapeType=Capsule, length=2, radius=1.5, shapeName=dyn_fronthairA_01_capsule, center=<1.8925 0 0.0001>, rotation=<rot 90 0 0>}]
 dyn_fronthairA_02 [{shapeType=Capsule, length=2, radius=1.5, shapeName=dyn_fronthairA_02_capsule, center=<1.8925 0 0.0001>, rotation=<rot 90 0 0>}]
 dyn_fronthairA_base [{shapeType=Capsule, length=0.505, radius=0.505, shapeName=dyn_fronthairA_base_capsule, rotation=<rot 0 -90 0>}]
 dyn_fronthairB_01 [{shapeType=Capsule, length=2, radius=1.5, shapeName=dyn_fronthairB_01_capsule, center=<1.8925 0 0.0001>, rotation=<rot 90 0 0>}]
 dyn_fronthairB_02 [{shapeType=Capsule, length=2, radius=1.5, shapeName=dyn_fronthairB_02_capsule, center=<1.8925 0 0.0001>, rotation=<rot 90 0 0>}]
 dyn_fronthairB_base [{shapeType=Capsule, length=0.505, radius=0.505, shapeName=dyn_fronthairB_base_capsule, rotation=<rot 0 -90 0>}]
 dyn_fronthairC_01 [{shapeType=Capsule, length=2, radius=1.5, shapeName=dyn_fronthairC_01_capsule, center=<1.8925 0 0.0001>, rotation=<rot 90 0 0>}]
 dyn_fronthairC_02 [{shapeType=Capsule, length=2, radius=1.5, shapeName=dyn_fronthairC_02_capsule, center=<1.8925 0 0.0001>, rotation=<rot 90 0 0>}]
 dyn_fronthairC_base [{shapeType=Capsule, length=0.505, radius=0.505, shapeName=dyn_fronthairC_base_capsule, rotation=<rot 0 -90 0>}]
 dyn_fronthairD_01 [{shapeType=Capsule, length=2, radius=1.5, shapeName=dyn_fronthairD_01_capsule, center=<1.8925 0 0.0001>, rotation=<rot 90 0 0>}]
 dyn_fronthairD_02 [{shapeType=Capsule, length=2, radius=1.5, shapeName=dyn_fronthairD_02_capsule, center=<1.8925 0 0.0001>, rotation=<rot 90 0 0>}]
 dyn_fronthairD_base [{shapeType=Capsule, length=0.505, radius=0.505, shapeName=dyn_fronthairD_base_capsule, rotation=<rot 0 -90 0>}]
 head [
  {shapeType=Sphere, radius=7.25, shapeName=head_sphere, center=<3.5461 1.9825 0>}
  {shapeType=Capsule, length=7.75, radius=5.75, rotation=<rot 90 0 0>}
 ]
 neck_01 [{shapeType=Capsule, length=0, radius=5.755, shapeName=neck_01_capsule, rotation=<rot 90 0 0>}]
}
constraints [
 {
  bone1Name=dyn_backhairA_base
  bone2Name=head
  swing1LimitDegrees=45
  swing1Motion=Free
  swing2LimitDegrees=45
  swing2Motion=Free
  twistLimitDegrees=45
  twistMotion=Free
 }
 {bone1Name=dyn_backhairA_01, bone2Name=dyn_backhairA_base, swing1LimitDegrees=6, swing2LimitDegrees=10, twistLimitDegrees=5}
 {bone1Name=dyn_backhairA_02, bone2Name=dyn_backhairA_01, swing1LimitDegrees=6, swing2LimitDegrees=10, twistLimitDegrees=5}
 {bone1Name=dyn_backhairA_03, bone2Name=dyn_backhairA_02, swing1LimitDegrees=6, swing2LimitDegrees=10, twistLimitDegrees=5}
 {bone1Name=dyn_backhairA_04, bone2Name=dyn_backhairA_03, swing1LimitDegrees=6, swing2LimitDegrees=10, twistLimitDegrees=5}
 {
  bone1Name=dyn_backhairB_base
  bone2Name=head
  swing1LimitDegrees=45
  swing1Motion=Free
  swing2LimitDegrees=45
  swing2Motion=Free
  twistLimitDegrees=45
  twistMotion=Free
 }
 {bone1Name=dyn_backhairB_01, bone2Name=dyn_backhairB_base, swing1LimitDegrees=2, swing2LimitDegrees=2.5, twistLimitDegrees=5}
 {bone1Name=dyn_backhairB_02, bone2Name=dyn_backhairB_01, swing1LimitDegrees=6, swing2LimitDegrees=6, twistLimitDegrees=5}
 {bone1Name=dyn_backhairB_03, bone2Name=dyn_backhairB_02, swing1LimitDegrees=6, swing2LimitDegrees=6, twistLimitDegrees=5}
 {bone1Name=dyn_backhairB_04, bone2Name=dyn_backhairB_03, swing1LimitDegrees=6, swing2LimitDegrees=6, twistLimitDegrees=5}
 {
  bone1Name=dyn_backhairC_base
  bone2Name=head
  swing1LimitDegrees=45
  swing1Motion=Free
  swing2LimitDegrees=45
  swing2Motion=Free
  twistLimitDegrees=45
  twistMotion=Free
 }
 {bone1Name=dyn_backhairC_01, bone2Name=dyn_backhairC_base, swing1LimitDegrees=10, swing2LimitDegrees=10, twistLimitDegrees=5}
 {bone1Name=dyn_backhairC_02, bone2Name=dyn_backhairC_01, swing1LimitDegrees=10, swing2LimitDegrees=10, twistLimitDegrees=5}
 {bone1Name=dyn_backhairC_03, bone2Name=dyn_backhairC_02, swing1LimitDegrees=10, swing2LimitDegrees=10, twistLimitDegrees=5}
 {bone1Name=dyn_backhairC_04, bone2Name=dyn_backhairC_03, swing1LimitDegrees=10, swing2LimitDegrees=10, twistLimitDegrees=5}
 {
  bone1Name=dyn_backhairD_base
  bone2Name=head
  swing1LimitDegrees=45
  swing1Motion=Free
  swing2LimitDegrees=45
  swing2Motion=Free
  twistLimitDegrees=45
  twistMotion=Free
 }
 {bone1Name=dyn_backhairD_01, bone2Name=dyn_backhairD_base, swing1LimitDegrees=10, swing2LimitDegrees=10, twistLimitDegrees=5}
 {bone1Name=dyn_backhairD_02, bone2Name=dyn_backhairD_01, swing1LimitDegrees=10, swing2LimitDegrees=10, twistLimitDegrees=5}
 {bone1Name=dyn_backhairD_03, bone2Name=dyn_backhairD_02, swing1LimitDegrees=10, swing2LimitDegrees=10, twistLimitDegrees=5}
 {bone1Name=dyn_backhairD_04, bone2Name=dyn_backhairD_03, swing1LimitDegrees=10, swing2LimitDegrees=10, twistLimitDegrees=5}
 {
  bone1Name=dyn_backhairE_base
  bone2Name=head
  swing1LimitDegrees=45
  swing1Motion=Free
  swing2LimitDegrees=45
  swing2Motion=Free
  twistLimitDegrees=45
  twistMotion=Free
 }
 {bone1Name=dyn_backhairE_01, bone2Name=dyn_backhairE_base, swing1LimitDegrees=10, swing2LimitDegrees=10, twistLimitDegrees=5}
 {bone1Name=dyn_backhairE_02, bone2Name=dyn_backhairE_01, swing1LimitDegrees=10, swing2LimitDegrees=10, twistLimitDegrees=5}
 {bone1Name=dyn_backhairE_03, bone2Name=dyn_backhairE_02, swing1LimitDegrees=10, swing2LimitDegrees=10, twistLimitDegrees=5}
 {bone1Name=dyn_backhairE_04, bone2Name=dyn_backhairE_03, swing1LimitDegrees=10, swing2LimitDegrees=10, twistLimitDegrees=5}
 {
  bone1Name=dyn_fronthairA_base
  bone2Name=head
  swing1LimitDegrees=45
  swing1Motion=Free
  swing2LimitDegrees=45
  swing2Motion=Free
  twistLimitDegrees=45
  twistMotion=Free
 }
 {bone1Name=dyn_fronthairA_01, bone2Name=dyn_fronthairA_base, swing1LimitDegrees=10, swing2LimitDegrees=10, twistLimitDegrees=5}
 {bone1Name=dyn_fronthairA_02, bone2Name=dyn_fronthairA_01, swing1LimitDegrees=10, swing2LimitDegrees=10, twistLimitDegrees=5}
 {
  bone1Name=dyn_fronthairB_base
  bone2Name=head
  swing1LimitDegrees=45
  swing1Motion=Free
  swing2LimitDegrees=45
  swing2Motion=Free
  twistLimitDegrees=45
  twistMotion=Free
 }
 {bone1Name=dyn_fronthairB_01, bone2Name=dyn_fronthairB_base, swing1LimitDegrees=10, swing2LimitDegrees=10, twistLimitDegrees=5}
 {bone1Name=dyn_fronthairB_02, bone2Name=dyn_fronthairB_01, swing1LimitDegrees=10, swing2LimitDegrees=10, twistLimitDegrees=5}
 {
  bone1Name=dyn_fronthairC_base
  bone2Name=head
  swing1LimitDegrees=45
  swing1Motion=Free
  swing2LimitDegrees=45
  swing2Motion=Free
  twistLimitDegrees=45
  twistMotion=Free
 }
 {bone1Name=dyn_fronthairC_01, bone2Name=dyn_fronthairC_base, swing1LimitDegrees=10, swing2LimitDegrees=10, twistLimitDegrees=5}
 {bone1Name=dyn_fronthairC_02, bone2Name=dyn_fronthairC_01, swing1LimitDegrees=10, swing2LimitDegrees=10, twistLimitDegrees=5}
 {
  bone1Name=dyn_fronthairD_base
  bone2Name=head
  swing1LimitDegrees=45
  swing1Motion=Free
  swing2LimitDegrees=45
  swing2Motion=Free
  twistLimitDegrees=45
  twistMotion=Free
 }
 {bone1Name=dyn_fronthairD_01, bone2Name=dyn_fronthairD_base, swing1LimitDegrees=10, swing2LimitDegrees=10, twistLimitDegrees=5}
 {bone1Name=dyn_fronthairD_02, bone2Name=dyn_fronthairD_01, swing1LimitDegrees=10, swing2LimitDegrees=10, twistLimitDegrees=5}
 {
  bone1Name=head
  bone2Name=neck_01
  swing1LimitDegrees=45
  swing1Motion=Free
  swing2LimitDegrees=45
  swing2Motion=Free
  twistLimitDegrees=45
  twistMotion=Free
 }
]
